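# Two-sided subsurface Nanite foliage material for Asteria bushes.
# World position offset combines wind with camera-proximity shrinking of randomly selected leaves toward instance-space pivots decoded from vertex color.
# Leaf shading blends texture, vertex-color tints, primitive/per-instance colors and random hue shifts, then passes through snow, hit-wobble, Curie fallback, move-tool and impostor-bake stages.
# A static switch adds UV-masked holiday bulbs with time-cycled emissive hues and per-lamp uniform fading.
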Material /Game/Environments/Asteria/Foliage/Bushes/Bush/M_Asteria_Bush_Nanite
graph {
 properties {
  bUseMaterialAttributes=true
  bUsedWithInstancedStaticMeshes=true
  bUsedWithNanite=true
  pixelDepthOffsetMode=PDOM_Legacy
  shadingModel=MSM_Subsurface
  twoSided=true
  usedShadingModels=Subsurface
 }
 nodes [
  ActorPositionWS()
  Constant()
  MaterialFunctionCall() {materialFunction=&DebugTimeSine}
  n1 = VertexColor()
  n2 = MaterialFunctionCall(Float3=@n1) {materialFunction=&BreakOutFloat3Components}
  n3 = MaterialFunctionCall(X=@n2.R, Y=@n2.R, Z=@n2.R) {materialFunction=&MakeFloat3}
  Multiply(A=ConstantBiasScale(None=@n3.Result) {bias=-0.5, scale=2}) {constB=200}
  n4 = MaterialFunctionCall(Float3=VertexColor()) {materialFunction=&BreakOutFloat3Components}
  n5 = MaterialFunctionCall(X=@n4.R, Y=@n4.R, Z=@n4.R) {materialFunction=&MakeFloat3}
  n6 = VectorParameter() {group=NaniteWPO, parameterName=PivotOffsetMatch, defaultValue=<rgba -0.5 -0.5 -0.25 1>}
  n7 = TransformPosition(None=Multiply(A=Add(A=@n5.Result, B=@n6.RGB), B=ScalarParameter() {defaultValue=1100, group=NaniteWPO, parameterName=ScaleMatch})) {transformSourceType=TRANSFORMPOSSOURCE_Instance, transformType=TRANSFORMPOSSOURCE_World}
  n8 = MaterialFunctionCall!(Position=@n7, IsTrees, IsGrass, IsMediumPlant, IsMediumPlant_2024Anim, IsMediumPlant2ndLevel, MediumPlant_BasePivotPostion) {materialFunction=/Game/Environments/Asteria/Foliage/Wind/MF_WindActorData}
  n9 = Multiply(A=Subtract(A=@n8.WindIntensitySingleSample) {constB=0.25}, B=@n8.WindI_Direction_WithNoise)
  n10 = LinearInterpolate(
   A=ScalarParameter() {defaultValue=32, parameterName=Wind_BigBush_Storm}
   B=ScalarParameter() {defaultValue=8, parameterName=Wind_BigBush}
   Alpha=@n8.Speed 0-1
  )
  n11 = WorldPosition() {worldPositionShaderOffset=WPT_ExcludeAllShaderOffsets}
  n12 = CameraPositionWS()
  n13 = WorldPosition()
  n14 = StaticSwitchParameter(
   True=LinearInterpolate(A=@n13.XYZ, B=@n7, Alpha=NamedRerouteUsage()) {desc="Ensure holiday lamps fade uniformly per lamp, not per vertex"}
   False=@n13.XYZ
  ) {parameterName=HasHolidayLights}
  n15 = Multiply(
   A=ScalarParameter() {defaultValue=0.25, group=NaniteWPO, parameterName=Scaling_Far}
   B=MaterialFunctionCall() {materialFunction=&ObjectScale}.Scale X
  )
  n16 = Multiply(
   A=ScalarParameter() {defaultValue=0.5, group=NaniteWPO, parameterName=Scaling_Close}
   B=MaterialFunctionCall() {materialFunction=&ObjectScale}.Scale X
  )
  n17 = CameraPositionWS()
  n18 = NamedRerouteUsage()
  n19 = SphereMask(
   A=Divide(A=Distance(A=@n17, B=@n18)) {constB=1000}
   B=Constant()
   Radius=Multiply(
    A=ScalarParameter() {defaultValue=0.6, group=NaniteWPO, parameterName=Scaling_Radius}
    B=MaterialFunctionCall() {materialFunction=&ObjectScale}.Scale X
   )
  ) {attenuationRadius=0.3, hardnessPercent=0}
  n20 = Sine(None=Multiply(A=@n4.R, B=Add(A=@n4.G, B=@n4.G)))
  n21 = ScalarParameter() {defaultValue=659.55, group=NaniteWPO, parameterName=RandomSeed}
  n22 = Frac(None=Subtract(A=Multiply(A=@n20, B=@n21)) {constB=0.694})
  n23 = Step(Y=@n22, X=ScalarParameter() {defaultValue=0.15, group=NaniteWPO, parameterName=UnhideLeafAmount}) {constX=0.1, constY=0.85}
  n24 = SphereMask(
   A=Divide(A=Distance(A=@n12, B=@n14), B=ScalarParameter() {defaultValue=1000, group=NaniteWPO, parameterName=DistanceDivisor}) {constB=1000}
   B=Constant()
   Radius=LinearInterpolate(A=@n15, B=LinearInterpolate(A=@n15, B=@n16, Alpha=@n19), Alpha=OneMinus(None=@n23))
  ) {attenuationRadius=0.1, hardnessPercent=50}
  n25 = Multiply(
   A=Subtract(A=@n7, B=@n11.XYZ)
   B=SmoothStep(Max=ScalarParameter() {defaultValue=0.5, group=NaniteWPO, parameterName=LeafSizeMin}, Value=Power(Base=@n24) {constExponent=0.5}) {constMax=0.5}
  )
  n26 = OneMinus(None=Step(Y=@n22) {constX=0, constY=0.85})
  NamedRerouteDeclaration(None=Add(A=Multiply(A=@n9, B=@n10), B=Multiply(A=@n25, B=@n26))) {name=WPO, nodeColor=<rgba 1 0 0.6328 1>}
  n27 = MaterialFunctionCall(Float3=ActorPositionWS()) {materialFunction=&BreakOutFloat3Components}
  n28 = Sine(None=Divide(A=@n27.R, B=Constant() {r=13000}))
  n29 = Sine(None=Divide(A=@n27.G, B=Constant() {r=11000}))
  NamedRerouteDeclaration(None=Divide(A=Add(A=@n28, B=@n29))) {name=PerActorRandom, nodeColor=<rgba 0 1 0.7891 1>}
  n30 = Add(
   A=MaterialFunctionCall!(Float In=ScalarParameter() {defaultValue=0.4, group=Shading, parameterName=SubSurface SSS Opacity}, Vector_In, Index=Constant() {r=3}) {materialFunction=/Game/Environments/Asteria/Foliage/Trees/MasterMaterials/MaterialFunctions/MF_AsteriaTree_ArrayAttribute}.Float Out
   B=MaterialFunctionCall(Float In=ScalarParameter() {defaultValue=0.5, group=Shading, parameterName=SubSurface SSS Opacity Variation}, Index=Constant() {r=4}) {materialFunction=&MF_AsteriaTree_ArrayAttribute}.Float Out
  )
  n31 = NamedRerouteUsage()
  NamedRerouteDeclaration(None=LinearInterpolate(A=Constant() {desc=Shadows Off SSS Opacity, r=0.875}, B=Multiply(A=@n30, B=@n31), Alpha=NamedRerouteUsage())) {name=SubSurfaceOpacity, nodeColor=<rgba 0 1 0.9062 1>}
  n32 = ScalarParameter() {defaultValue=1, group=Shading, parameterName=Roughness_LeafBottom}
  n33 = Constant() {r=8}
  n34 = ScalarParameter() {defaultValue=0.2, parameterName=Roughness}
  n35 = Constant() {r=7}
  n36 = TwoSidedSign()
  n37 = LinearInterpolate(
   A=LinearInterpolate(
    A=MaterialFunctionCall(Float In=@n32, Index=@n33) {materialFunction=&MF_AsteriaTree_ArrayAttribute}.Float Out
    B=MaterialFunctionCall(Float In=@n34, Index=@n35) {materialFunction=&MF_AsteriaTree_ArrayAttribute}.Float Out
    Alpha=Saturate(None=@n36)
   )
   B=MaterialFunctionCall(Float In=ScalarParameter() {defaultValue=1.5227, group=Shading, parameterName=Roughness_At_Distance}, Index=Constant() {r=32}) {materialFunction=&MF_AsteriaTree_ArrayAttribute}.Float Out
  ) {constAlpha=0}
  n38 = NamedRerouteUsage()
  NamedRerouteDeclaration(None=LinearInterpolate(A=Constant() {desc=RoughnessReflections Disabled, r=0.8}, B=Multiply(A=@n37, B=@n38), Alpha=NamedRerouteUsage())) {name=Roughness, nodeColor=<rgba 0 0.2031 1 1>}
  n39 = ScalarParameter() {defaultValue=0.4622, group=Shading, parameterName=Specular_LeafBottom}
  n40 = Constant() {r=5}
  n41 = ScalarParameter() {defaultValue=0.507, group=Shading, parameterName=Specular}
  n42 = Constant() {r=6}
  n43 = TwoSidedSign()
  n44 = Multiply(
   A=NamedRerouteUsage()
   B=LinearInterpolate(
    A=MaterialFunctionCall(Float In=@n39, Index=@n40) {materialFunction=&MF_AsteriaTree_ArrayAttribute}.Float Out
    B=MaterialFunctionCall(Float In=@n41, Index=@n42) {materialFunction=&MF_AsteriaTree_ArrayAttribute}.Float Out
    Alpha=Saturate(None=@n43)
   )
  )
  n45 = ScalarParameter() {defaultValue=0.125, group=Shading, parameterName=Specular_PerLeaf_Random_Low}
  n46 = Constant() {r=36}
  n47 = ScalarParameter() {defaultValue=1.5, group=Shading, parameterName=Specular_PerLeaf_Random_High}
  n48 = Constant() {r=35}
  n49 = Saturate(None=LinearInterpolate(
    A=MaterialFunctionCall(Float In=@n45, Index=@n46) {materialFunction=&MF_AsteriaTree_ArrayAttribute}.Float Out
    B=MaterialFunctionCall(Float In=@n47, Index=@n48) {materialFunction=&MF_AsteriaTree_ArrayAttribute}.Float Out
   ) {constAlpha=1})
  NamedRerouteDeclaration(None=LinearInterpolate(A=Constant() {desc=Specular Reflections Disabled, r=0.1}, B=Multiply(A=@n44, B=@n49), Alpha=NamedRerouteUsage())) {name=Specular, nodeColor=<rgba 0.1484 1 0 1>}
  n50 = Multiply(A=Constant() {r=0.1}, B=NamedRerouteUsage())
  NamedRerouteDeclaration(None=StaticSwitch(True=Constant() {r=1}, False=Add(A=@n50), Value=NamedRerouteUsage()) {desc=For ImpostorBakeing}) {name=ValueVariation, nodeColor=<rgba 0 0.0625 1 1>}
  n51 = VectorParameter() {bUseCustomPrimitiveData=true, parameterName=Color1_Base_primdata}
  n52 = PerInstanceCustomData!(DefaultValue) {dataIndex=3}
  n53 = NamedRerouteDeclaration(None=Add(A=@n51.A, B=@n52)) {name=PrimDataBlendAlpha, nodeColor=<rgba 0 1 0.3203 1>}
  n54 = LinearInterpolate(
   A=MaterialFunctionCall(Vector_In=VectorParameter() {group=Color, parameterName=Color1_Base, defaultValue=<rgba 1 1 1 1>}.RGB, Index=Constant() {r=24}) {materialFunction=&MF_AsteriaTree_ArrayAttribute}.Vector_Out
   B=Add(A=PerInstanceCustomData3Vector!(DefaultValue), B=@n51.RGB)
   Alpha=@n53
  )
  n55 = VectorParameter() {group=Color VertexColor Tints, parameterName=Green Vertex Channel Color, sortPriority=4, defaultValue=<rgba 1 1 1 1>}
  n56 = VectorParameter() {bUseCustomPrimitiveData=true, desc=20 not used, parameterName=Color2_Base_primdata, primitiveDataIndex=17, defaultValue=<rgba 0 0 0 1>}
  n57 = LinearInterpolate(A=@n55.RGB, B=Add(A=PerInstanceCustomData3Vector() {dataIndex=17}, B=@n56.RGB), Alpha=@n53)
  n58 = Saturate(None=LinearInterpolate(
    A=ScalarParameter() {group=Color VertexColor Tints, parameterName=Green Vertex Channel Mask Value LOW, sortPriority=5}
    B=ScalarParameter() {group=Color VertexColor Tints, parameterName=Green Vertex Channel Mask Value HIGH, sortPriority=5}
    Alpha=VertexColor().G
   ))
  NamedRerouteDeclaration(None=StaticSwitch(True=Constant3Vector() {constant=<rgba 1 1 1 0>}, False=LinearInterpolate(A=@n54, B=@n57, Alpha=@n58), Value=NamedRerouteUsage()) {desc=For ImpostorBakeing}) {name=LeafColor, nodeColor=<rgba 0 1 0.5547 1>}
  n59 = ScalarParameter() {defaultValue=1, group=Shading, parameterName=Brightness}
  n60 = Constant() {r=20}
  n61 = NamedRerouteUsage()
  n62 = TextureSampleParameter2D() {
   group=Textures
   parameterName=Diffuse
   samplerSource=SSM_Wrap_WorldGroupSettings
   texture=/Game/Characters/Fortress_Character/Textures/T_All_White
  }
  n63 = Multiply(A=NamedRerouteUsage(), B=@n62.RGB)
  n64 = MaterialFunctionCall!(
   Leaves Input=Multiply(
    A=NamedRerouteUsage()
    B=MaterialFunctionCall(Hue Shift Percentage=@n61, Texture=@n63) {materialFunction=&HueShift}.Result
   )
   Trunk_Input
   UseCustomPrimValue=NamedRerouteUsage()
  ) {materialFunction=/Game/Environments/Asteria/Foliage/Trees/MasterMaterials/MaterialFunctions/MF_Asteria_Tree_VertexColorBlend}
  NamedRerouteDeclaration(None=Multiply(
    A=MaterialFunctionCall(Float In=@n59, Index=@n60) {materialFunction=&MF_AsteriaTree_ArrayAttribute}.Float Out
    B=@n64.Leaves Output
   )) {name=BaseColor, nodeColor=<rgba 0 1 0.4844 1>}
  n65 = TextureSampleParameter2D() {
   group=Textures
   parameterName=SSSR
   samplerSource=SSM_Wrap_WorldGroupSettings
   samplerType=SAMPLERTYPE_Masks
   texture=&T_WhiteMask
   channelNames {a="", b=Roughness, g=SubSurfaceScatter, r=Specular}
  }
  n66 = MaterialFunctionCall(Float3=@n65.RGB) {materialFunction=&BreakOutFloat3Components}
  NamedRerouteDeclaration(None=@n66.G) {name=SSSTextureInput, nodeColor=<rgba 0.8516 0 1 1>}
  NamedRerouteDeclaration(None=@n66.B) {name=RoughnessTextureInput, nodeColor=<rgba 0 1 0.5781 1>}
  NamedRerouteDeclaration(None=@n52) {name=IsPerIstanceData, nodeColor=<rgba 1 0.2109 0 1>}
  NamedRerouteDeclaration(None=@n22) {name=PerLeafRandom, nodeColor=<rgba 1 0 0.6308 1>}
  NamedRerouteDeclaration(None=TransformPosition(None=Constant3Vector()) {transformSourceType=TRANSFORMPOSSOURCE_Instance, transformType=TRANSFORMPOSSOURCE_World}) {name=ActorLocation, nodeColor=<rgba 0 0.2734 1 1>}
  n67 = CollectionParameter() {collection=&FortniteMaterialParameters, parameterName=ReflectionsEnabled}
  NamedRerouteDeclaration(None=@n67) {name=ReflectionsEnabled, nodeColor=<rgba 1 0 0.6562 1>}
  NamedRerouteDeclaration(None=@n66.R) {name=SpecularMaskTexture, nodeColor=<rgba 1 0.8672 0 1>}
  n68 = TextureSampleParameter2D() {
   group=Textures
   parameterName=Normal
   samplerSource=SSM_Wrap_WorldGroupSettings
   samplerType=SAMPLERTYPE_Normal
   texture=/Engine/EngineMaterials/BaseFlattenNormalMap
  }
  n69 = ScalarParameter() {defaultValue=0.4, group=Shading, parameterName=NormalFlatten}
  n70 = Constant() {r=1}
  n71 = Normalize(VectorInput=MaterialFunctionCall(
    Normal=@n68.RGB
    Flatness=MaterialFunctionCall(Float In=@n69, Index=@n70) {materialFunction=&MF_AsteriaTree_ArrayAttribute}.Float Out
   ) {materialFunction=&FlattenNormal}.Result)
  n72 = TwoSidedSign()
  NamedRerouteDeclaration(None=StaticSwitchParameter(True=Multiply(A=@n71, B=@n72), False=@n71) {parameterName=FlipBackfaceNormals}) {name=Normal, nodeColor=<rgba 0.3727 0.4306 1 1>}
  n73 = Constant()
  n74 = Multiply(A=NamedRerouteUsage(), B=Constant() {desc=Fixed value For Wold HueVariation, r=0.0125})
  n75 = NamedRerouteUsage()
  n76 = ScalarParameter() {defaultValue=-0.0645, group=Color VertexColor Tints, parameterName=Red Vertex Channel Hue Shift PerLeaf Random, sortPriority=1}
  n77 = ScalarParameter() {bUseCustomPrimitiveData=true, defaultValue=-0.03, group=Color, parameterName=Hue Variation_primdata, primitiveDataIndex=4}
  n78 = PerInstanceCustomData() {constDefaultValue=-0.03, dataIndex=4}
  n79 = NamedRerouteUsage()
  n80 = MaterialFunctionCall(Float In=LinearInterpolate(A=@n76, B=LinearInterpolate(A=@n77, B=@n78, Alpha=@n79), Alpha=NamedRerouteUsage()), Index=Constant()) {materialFunction=&MF_AsteriaTree_ArrayAttribute}
  n81 = NamedRerouteUsage()
  n82 = Multiply(A=Frac(None=Add(B=@n81) {constA=0.5}), B=@n81)
  NamedRerouteDeclaration(None=Add(A=StaticSwitch(True=@n73, False=@n74, Value=@n75) {desc=For ImpostorBakeing}, B=Multiply(A=@n80.Float Out, B=@n82))) {name=HueShift, nodeColor=<rgba 0 1 0.7891 1>}
  n83 = Multiply(
   A=@n64.Leaves Output
   B=MaterialFunctionCall(Float In=ScalarParameter() {defaultValue=0.2, group=Shading, parameterName=SubSurface SSSColor}, Index=Constant() {r=21}) {materialFunction=&MF_AsteriaTree_ArrayAttribute}.Float Out
  )
  n84 = Constant() {desc=Shaodows Off SSS Multiplier, r=0.875}
  NamedRerouteDeclaration(None=LinearInterpolate(A=Multiply(A=@n83, B=@n84) {constB=0.5}, B=@n83, Alpha=NamedRerouteUsage())) {name=SubSurfaceColor, nodeColor=<rgba 1 0.9844 0 1>}
  n85 = CollectionParameter() {collection=&FortniteMaterialParameters, parameterName=ShadowsEnabled}
  NamedRerouteDeclaration(None=@n85) {name=ShadowsEnabled, nodeColor=<rgba 0.3828 0 1 1>}
  n86 = Constant3Vector() {constant=<rgba 1 0 0 0>}
  Normalize(VectorInput=CrossProduct(A=@n8.WindI_Direction, B=@n86))
  RotateAboutAxis(
   NormalizedRotationAxis=@n86
   RotationAngle=Multiply(A=@n8.WindIntensitySingleSample, B=ScalarParameter() {defaultValue=0.456, parameterName=RotationTest})
   PivotPoint=@n7
   Position=@n11.XYZ
  )
  ScalarParameter() {defaultValue=13000, group=Color Variation WorldSpace, parameterName=Color Variation Distance WS, sortPriority=10}
  ScalarParameter() {defaultValue=11000, group=Color Variation WorldSpace, parameterName=Color Variation Distance WS2, sortPriority=10}
  ScalarParameter() {desc=debug, parameterName=Debug}
  StaticBool() {value=true}
  n87 = ShadingModel() {shadingModel=MSM_Subsurface}
  n88 = SetMaterialAttributes!(
   MaterialAttributes
   Base Color=NamedRerouteUsage()
   Roughness=NamedRerouteUsage()
   Specular=NamedRerouteUsage()
   Opacity=NamedRerouteUsage()
   Normal=NamedRerouteUsage()
   Subsurface Color=NamedRerouteUsage()
   World Position Offset=NamedRerouteUsage()
   Emissive Color
  ) {attributeSetTypes [
    #1
    #2
    #3
    #4
    #5
    #6
    #7
    #8
   ]}
  n89 = GetMaterialAttributes(None=MaterialFunctionCall!(
    In=@n88
    FlipTwoSidedSign=StaticBoolParameter() {defaultValue=true, group=Snow, parameterName=SnowFlipTwosidedNormal}
    IsTangentSpaceInput=StaticBool() {value=true}
    HasBackFlipFaceNormals=StaticBoolParameter() {parameterName=UseFlipBackFaceNormals}
   ) {materialFunction=/Game/Packages/Fortnite_UniqueMaterials_1/MaterialFunctions/Trees/MF_Tree_Snow}.Result) {attributeGetTypes [#7]}
  n90 = SetMaterialAttributes!(
   MaterialAttributes=@n89.MaterialAttributes
   World Position Offset=Add(
    A=@n89.World Position Offset
    B=MaterialFunctionCall() {materialFunction=&MF_FN_DestructibleDamage}.DamageAnimWPO
   )
  ) {attributeSetTypes [#7]}
  n91 = StaticSwitchParameter(True=@n90, False=StaticSwitchParameter(True=@n90, False=@n89.MaterialAttributes) {group=Trunk, parameterName=IsTrunk}) {group=Trunk, parameterName=ForceEnable_HitFXWobble}
  n92 = MaterialFunctionCall(MA In=MaterialFunctionCall!(
    In=ShadingPathSwitch(Default=@n91, Deferred=@n91, Forward=@n89.MaterialAttributes, Mobile=@n89.MaterialAttributes)
    UseSphericalMappingOverride=StaticBool() {value=true}
   ) {materialFunction=/Game/Packages/Fortress_SharedMaterials/Base_Material_Functions/MF_CurieFallbackEffect}.Result) {materialFunction=&MF_CreativeMoveTool}
  n93 = StaticSwitch(
   True=MaterialFunctionCall!(Material Attributes=@n92.MA Out, IsTangentSpaceNormal=StaticBool() {desc=Don't change, value=true}, In=Constant() {r=1}) {materialFunction=/Game/ImpostorBaker/MaterialFunctions/ImposterCaptureSwitch}.Result
   False=@n92.MA Out
   Value=NamedRerouteDeclaration(None=StaticBoolParameter() {parameterName=EnableForImpostorBake}) {name=IsImpostorBake, nodeColor=<rgba 1 0.3984 0 1>}
  )
  n94 = TextureCoordinate()
  n95 = ComponentMask(None=@n94) {r=true}
  n96 = Saturate(None=Floor(None=Subtract(A=@n95)))
  n97 = OneMinus(None=Fresnel() {baseReflectFraction=0})
  n98 = Divide(A=Floor(None=ComponentMask(None=@n94) {g=true})) {constB=16, desc=Mesh specific - UV0 vertical offset used as random seed}
  n99 = Time()
  n100 = Multiply(A=Floor(None=Multiply(A=@n99) {constB=0.9967})) {constB=0.25}
  n101 = Multiply(A=Frac(None=Add(A=@n98, B=@n100))) {constB=6}
  n102 = MaterialFunctionCall(Hue Shift Percentage=Divide(A=Floor(None=@n101)) {constB=6}, Texture=Constant3Vector() {constant=<rgba 1 0 0 0>}) {materialFunction=&HueShift}
  n103 = ScalarParameter() {defaultValue=30, parameterName=Brightness Bulbs}
  n104 = SetMaterialAttributes!(
   MaterialAttributes
   Base Color=Constant3Vector() {constant=<rgba 0.01 0.01 0.01 0>}
   Metallic=ScalarParameter() {parameterName=Metallic, sliderMax=1}
   Roughness=ScalarParameter() {defaultValue=0.2, parameterName=Roughness, sliderMax=1}
   Emissive Color=Multiply(A=Multiply(A=@n96, B=@n97), B=Multiply(A=@n102.Result, B=@n103))
   Shading Model=ShadingModel()
   Subsurface Color=Constant3Vector()
  ) {attributeSetTypes [
    #1
    #9
    #2
    #8
    #10
    #6
   ]}
  n105 = TextureCoordinate()
  n106 = Saturate(None=Floor(None=ComponentMask(None=@n105) {r=true}))
  StaticSwitchParameter(
   True=BlendMaterialAttributes(
    A=SetMaterialAttributes!(MaterialAttributes=@n93, Shading Model=@n87) {attributeSetTypes [#10]}
    B=@n104
    Alpha=NamedRerouteDeclaration(None=@n106) {name=LampMask, nodeColor=<rgba 0.2891 1 0 1>}
   ) {vertexAttributeBlendType=UseA}
   False=@n93
  ) {parameterName=HasHolidayLights}
  Subtract(A=WorldPosition().XYZ, B=ActorPositionWS())
  TransformPosition(None=Constant3Vector()) {transformSourceType=TRANSFORMPOSSOURCE_Instance, transformType=TRANSFORMPOSSOURCE_World}
  TransformPosition(None=Multiply(A=@n3.Result, B=MaterialFunctionCall() {materialFunction=&DebugTimeSine}.Result)) {transformSourceType=TRANSFORMPOSSOURCE_Instance}
  VertexInterpolator()
  VertexInterpolator(VS=TransformPosition(None=@n1) {transformSourceType=TRANSFORMPOSSOURCE_Instance})
 ]
}
